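# Material function for user-built props that outputs one animation mask for building, breaking and destroying sub-objects: black is in place, white is animated.
# Anim Order from SelfBuildingWalls2 drives wallAnim for the build phase and a threshold against TargetBreakNumber, scaled by BreakAnim.
# HeightLerp drives code-controlled total destruction from a compressed order gradient plus the break threshold, tying both animations together when destruction takes over.
# The results are summed and clamped.
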MaterialFunction /Game/Packages/User_Built_Props/Materials/Master/AnimationMF
graph {nodes [
  n1 = FunctionInput() {bUsePreviewValueAsDefault=true, inputName=BuildAnimPhase, inputType=FunctionInput_Scalar, previewValue=<33 0 0 1>}
  n2 = MaterialFunctionCall!(Number of sub objects, Manual Anim Grad, Sec mot Falloff rad) {desc="!!", materialFunction=/Game/Effects/Fort_Effects/Materials/Temp/SelfBuildingWalls2}
  n3 = FunctionInput() {bUsePreviewValueAsDefault=true, inputName=TotalNumberOfSubObjects, inputType=FunctionInput_Scalar, sortPriority=-1, previewValue=<34 0 0 1>}
  n4 = If(
   A=Subtract(A=@n3, B=@n2.Anim Order) {desc=reverse anim order}
   B=Subtract(A=@n3, B=FunctionInput() {bUsePreviewValueAsDefault=true, inputName=TargetBreakNumber, inputType=FunctionInput_Scalar, sortPriority=1}) {desc=reverse anim order}
   "A > B"=Constant()
   "A < B"=Constant() {r=1}
  )
  n5 = FunctionInput() {bUsePreviewValueAsDefault=true, inputName=BreakAnim, inputType=FunctionInput_Scalar, sortPriority=2}
  n6 = Clamp(None=Add(
    A=MaterialFunctionCall!(AnimPhase=@n1, AnimGrad=@n2.Anim Order) {materialFunction=/Game/Packages/User_Built_Props/wallAnim}.Result
    B=Multiply(A=@n4, B=@n5)
   ))
  n7 = Add(A=Multiply(A=Divide(A=@n2.Anim Order, B=@n3)) {constB=0.25}) {constB=0.75, desc=CompressDestruction}
  n8 = MaterialFunctionCall(
   Transition Phase=Clamp(None=FunctionInput() {bUsePreviewValueAsDefault=true, inputName=Code_totalDestructionAmount, inputType=FunctionInput_Scalar, sortPriority=3})
   Height Texture=Clamp(None=Add(A=@n4, B=@n7) {desc=ties the 2 animation together when break is usurped by total destruction via code})
  ) {materialFunction=&HeightLerp}
  FunctionOutput(None=Clamp(None=Add(A=@n6, B=@n8.Lerp Alpha No Contrast)) {desc="black = in place white= animated"}) {outputName=single}
 ]}
